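# Translucent lit fog sheet material, named for the castle base in the Magic Forest environment effects, with a flat grey base color parameter.
# Two panning samples of an atmospheric cloud noise at separate tilings blend to drive a density range, masked by a SphereMask gradient along U.
# Depth fade, near-camera fade and inverted Fresnel soften the resulting opacity value, which a time-of-day scalar function then scales.
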Material /Game/Effects/Fort_Effects/Effects/Environment/MagicForest/Materials/M_CastleBase_FogSheet
graph {
 properties {bAutomaticallySetUsageInEditor=false, blendMode=BLEND_Translucent, usedShadingModels=Default Lit}
 nodes [
  n1 = SphereMask(
   A=ComponentMask(None=TextureCoordinate()) {r=true}
   B=ScalarParameter() {defaultValue=1, parameterName=s}
   Radius=ScalarParameter() {defaultValue=0.9864}
   Hardness=ScalarParameter() {defaultValue=0.369, parameterName=Param_1}
  )
  n2 = VectorParameter() {defaultValue=<rgba 0.5768 0.9972 0 0>}
  n3 = Panner(
   Coordinate=Multiply(A=TextureCoordinate(), B=ComponentMask(None=@n2.RGB) {g=true, r=true})
   Time=Multiply(A=Time(), B=ScalarParameter() {defaultValue=0.01, parameterName=Param_3})
  ) {speedY=-1}
  n4 = VectorParameter() {parameterName=Param2, defaultValue=<rgba 0.4552 1 0 0>}
  n5 = Panner(
   Coordinate=Multiply(A=TextureCoordinate(), B=ComponentMask(None=@n4.RGB) {g=true, r=true})
   Time=Multiply(A=Time(), B=ScalarParameter() {defaultValue=0.01, parameterName=Param_32})
  ) {speedY=-1}
  n6 = LinearInterpolate(
   A=ScalarParameter() {defaultValue=0.3328, parameterName=Param_2}
   B=ScalarParameter() {defaultValue=6.5245, parameterName=Param_6}
   Alpha=LinearInterpolate(
    A=TextureSample(UVs=@n3) {texture=&AtmosphericCloudNoise01}.G
    B=TextureSample(UVs=@n5) {texture=&AtmosphericCloudNoise01}.G
   )
  )
  n7 = DepthFade(
   Opacity=Multiply(A=Multiply(A=@n1, B=@n6), B=ScalarParameter() {defaultValue=0.5, parameterName=Param_5})
   FadeDistance=ScalarParameter() {defaultValue=1500, parameterName=Param_4}
  )
  n8 = MaterialFunctionCall(Distance=ScalarParameter() {defaultValue=14500, parameterName=Param_7}) {materialFunction=&Near_Camera_Fade}
  n9 = Fresnel(ExponentIn=ScalarParameter() {defaultValue=0.5, parameterName=Param_8}) {exponent=2}
  Multiply(
   A=MaterialFunctionCall(
    Value_Morning=ScalarParameter() {defaultValue=0.45, parameterName="Value_Morning (S)"}
    Value_Day=ScalarParameter() {defaultValue=0.45, parameterName="Value_Day (S)"}
    Value_Evening=ScalarParameter() {defaultValue=0.45, parameterName="Value_Evening (S)"}
    Value_Night=ScalarParameter() {defaultValue=0.45, parameterName="Value_Night (S)"}
   ) {materialFunction=&Scalar_Time_of_Day}.Result
   B=Multiply(A=Multiply(A=@n7, B=@n8.Result), B=OneMinus(None=@n9))
  )
  VectorParameter() {parameterName=Base Color, defaultValue=<rgba 0.5896 0.5896 0.5896 0>}
 ]
}
