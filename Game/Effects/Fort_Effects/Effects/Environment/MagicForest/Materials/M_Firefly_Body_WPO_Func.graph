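# Material function that outputs World Position Offset for Magic Forest firefly bodies, giving them procedural flight motion.
# Time-scrolled curl noise, vertex-color-varied, provides drift (also exposed as a Velocity reroute) and adds to a red-channel-phased sine bob.
# Vertex colors, remapped to signed vectors, serve as pivots for a slow Z-axis rotation.
# A cross-product basis branch computes orientation offsets, and several leftover nodes are unconnected.
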MaterialFunction /Game/Effects/Fort_Effects/Effects/Environment/MagicForest/Materials/M_Firefly_Body_WPO_Func
graph {nodes [
  CameraPositionWS()
  CameraPositionWS()
  ConstantBiasScale(None=TextureSample() {samplerType=SAMPLERTYPE_LinearColor, texture=/Game/Effects/FogVolume/VT_CurlNoise_LOW}.RGB) {bias=-0.5, scale=2}
  ConstantBiasScale(None=VertexColor()) {bias=-0.5, scale=2}
  Divide(B=MaterialFunctionCall() {materialFunction=&ObjectScale}.Scale XYZ)
  n1 = VertexColor()
  n2 = Multiply(A=WorldPosition().XYZ, B=ConstantBiasScale(None=@n1) {bias=-0.5, scale=2})
  n3 = Multiply(A=Time(), B=VectorParameter() {parameterName=Param_1, defaultValue=<rgba 0 0 50 0>}.RGB)
  n4 = VectorNoise(World Position=Multiply(A=Add(A=@n2, B=@n3), B=ScalarParameter() {defaultValue=0.005, parameterName=Param_1})) {noiseFunction=VNF_CurlALU}
  n5 = ScalarParameter() {defaultValue=5}
  n6 = Time()
  n7 = ScalarParameter() {defaultValue=1, parameterName=Param_2}
  n8 = Sine(None=Add(A=@n1.R, B=Multiply(A=@n6, B=@n7)))
  n9 = Multiply(A=@n1.R, B=VectorParameter() {defaultValue=<rgba 1 0 1 0>}.RGB)
  n10 = Add(A=Add(A=Multiply(A=@n4, B=@n5), B=Multiply(A=@n8, B=@n9)))
  n11 = NamedRerouteUsage()
  n12 = TransformPosition(None=ConstantBiasScale(None=VertexColor()) {bias=-0.5, scale=2}) {transformType=TRANSFORMPOSSOURCE_World}
  n13 = Normalize(VectorInput=Subtract(A=Subtract(A=@n11, B=@n12), B=NamedRerouteUsage()))
  FunctionOutput(None=Add(A=Add(A=@n10, B=@n13), B=NamedRerouteUsage())) {outputName=WPO}
  MaterialFunctionCall() {materialFunction=&ObjectPivotPoint}
  MaterialFunctionCall!(Euler) {materialFunction=/Game/Tools/NewWaterfallGenerator/Material/MF_EulerToQUat}
  Multiply() {constB=-1}
  Multiply() {constB=-1}
  Multiply()
  Multiply()
  Multiply(B=ScalarParameter() {defaultValue=9500, parameterName=Param_5})
  Multiply() {constB=-1}
  NamedRerouteDeclaration(None=@n4) {name=Velocity, nodeColor=<rgba 0 1 0.9062 1>}
  n14 = Constant3Vector() {constant=<rgba 0 0 1 0>}
  n15 = Time()
  n16 = ScalarParameter() {defaultValue=0.05, parameterName=Param_4}
  n17 = ConstantBiasScale(None=VertexColor()) {bias=-0.5, scale=2}
  NamedRerouteDeclaration(None=RotateAboutAxis(
    NormalizedRotationAxis=@n14
    RotationAngle=Multiply(A=@n15, B=@n16)
    PivotPoint=TransformPosition(None=@n17) {transformSourceType=TRANSFORMPOSSOURCE_Instance, transformType=TRANSFORMPOSSOURCE_World}
    Position=WorldPosition().XYZ
   )) {name=RAA, nodeColor=<rgba 0.5703 1 0 1>}
  NamedRerouteUsage()
  ObjectPositionWS()
  ScalarParameter() {defaultValue=0.1, parameterName=Param_3}
  Subtract()
  Subtract(B=WorldPosition().XYZ)
  Subtract(B=TransformPosition() {transformSourceType=TRANSFORMPOSSOURCE_World})
  n18 = TransformPosition(None=WorldPosition().XYZ) {transformSourceType=TRANSFORMPOSSOURCE_World, transformType=TRANSFORMPOSSOURCE_Instance}
  n19 = VertexColor()
  n20 = Subtract(
   A=Transform(None=NamedRerouteUsage()) {transformSourceType=TRANSFORMSOURCE_Local}
   B=TransformPosition(None=ConstantBiasScale(None=@n19) {bias=-0.5, scale=2}) {transformSourceType=TRANSFORMPOSSOURCE_Instance, transformType=TRANSFORMPOSSOURCE_World}
  )
  n21 = ObjectPositionWS()
  n22 = Normalize(VectorInput=Subtract(A=NamedRerouteUsage(), B=Subtract(A=@n20, B=@n21)))
  n23 = Normalize(VectorInput=CrossProduct(A=@n22, B=@n14))
  n24 = Multiply(A=@n23) {constB=-1}
  n25 = Normalize(VectorInput=CrossProduct(A=@n23, B=@n22))
  Transform(None=Subtract(
    A=MaterialFunctionCall(VectorToTransform=@n18, BasisX=@n22, BasisY=@n24, BasisZ=@n25) {materialFunction=&Transform3x3Matrix}.Transformed Vector
    B=@n18
   )) {transformSourceType=TRANSFORMSOURCE_Local}
  VertexColor()
 ]}
