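# Translucent unlit glow material for Magic Forest fireflies, with instanced static mesh usage enabled.
# A Glow Color parameter feeds an emissive adjustment function, while a shared firefly body WPO function is used for vertex motion.
# An opacity-style mask multiplies particle alpha, a high-exponent Fresnel, depth fade and a clamped sine pulse.
# The TimeOffset dynamic parameter shifts the pulse phase, allowing fireflies to blink out of sync.
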Material /Game/Effects/Fort_Effects/Effects/Environment/MagicForest/Materials/M_FireflyGlow
graph {
 properties {
  bUsedWithInstancedStaticMeshes=true
  blendMode=BLEND_Translucent
  pixelDepthOffsetMode=PDOM_Legacy
  shadingModel=MSM_Unlit
  usedShadingModels=Unlit
 }
 nodes [
  MaterialFunctionCall(Emissive=VectorParameter() {parameterName=Glow Color, defaultValue=<rgba 0.4374 0 1 1>}.RGB) {materialFunction=&MF_FX_EmissiveAdjustments}
  MaterialFunctionCall() {materialFunction=&/M_Firefly_Body_WPO_Func}
  n1 = ParticleColor()
  n2 = MaterialFunctionCall(Exponent=Constant() {r=24}) {materialFunction=&MF_Fresnel_FX}
  n3 = Constant() {r=6}
  n4 = Add(A=Time(), B=DynamicParameter() {defaultValue=<rgba 0 1 1 0>, paramNames [Param1,Param2,Param3,TimeOffset]}.TimeOffset)
  n5 = Add(A=Sine(None=Multiply(A=@n4) {constB=0.35})) {constB=0.2}
  Multiply(A=Multiply(A=Multiply(A=@n1.A, B=@n2.Fresnel), B=DepthFade(FadeDistance=@n3)), B=Clamp(None=Multiply(A=@n5) {constB=2}))
 ]
}
