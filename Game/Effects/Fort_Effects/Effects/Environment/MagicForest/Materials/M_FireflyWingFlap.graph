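# Translucent, two-sided lit material for firefly mesh particles that animates flapping wings.
# Texture color and alpha are multiplied by particle color, with motion-blur fade also applied to alpha.
# A sine of time plus per-particle random, scaled by flap speed and amount, drives an instance-space offset masked by an inverted sphere mask on instance-bounds UVW.
# A shared firefly body function supplies additional world position offset.
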Material /Game/Effects/Fort_Effects/Effects/Environment/MagicForest/Materials/M_FireflyWingFlap
graph {
 properties {
  bEnableResponsiveAA=true
  bUsedWithInstancedStaticMeshes=true
  bUsedWithNiagaraMeshParticles=true
  blendMode=BLEND_Translucent
  pixelDepthOffsetMode=PDOM_Legacy
  previewMesh=/Game/Effects/Fort_Effects/Mesh/Fireflies/SM_Firefly_NoShell
  translucencyLightingMode=TLM_VolumetricPerVertexNonDirectional
  twoSided=true
  usedShadingModels=Default Lit
 }
 nodes [
  Add(B=MaterialFunctionCall() {materialFunction=&/M_Firefly_Body_WPO_Func}.WPO)
  DynamicParameter() {parameterIndex=1, defaultValue=<rgba 0 0 0 1>, paramNames [FlapOffset,ZFlapScalar,FlapSpeed,Param4]}
  n1 = TextureSample() {texture=/Game/Effects/Fort_Effects/Textures/Fireflies/T_Firefly_Mesh_01}
  n2 = ParticleColor()
  Multiply(A=@n1.RGB, B=@n2.RGB)
  Multiply(A=Multiply(A=@n1.A, B=@n2.A), B=ParticleMotionBlurFade())
  Multiply()
  Multiply()
  n3 = TransformPosition(None=ObjectPositionWS()) {transformSourceType=TRANSFORMPOSSOURCE_World, transformType=TRANSFORMPOSSOURCE_Instance}
  n4 = VectorParameter() {defaultValue=<rgba 0 0 -1 0>}
  n5 = MaterialFunctionCall() {materialFunction=/Game/Effects/Niagara/MasterMaterials/Functions/InstanceBounds_0-1_UVW}
  n6 = ScalarParameter() {defaultValue=0.5, parameterName=Param_1}
  n7 = ScalarParameter() {defaultValue=0.3, parameterName=Param_2}
  n8 = ScalarParameter() {parameterName=Param_3}
  n9 = Power(Base=OneMinus(None=SphereMask(A=@n5.G, B=@n6, Radius=@n7, Hardness=@n8)), Exp=ScalarParameter() {defaultValue=8})
  n10 = Add(A=Time(), B=ParticleRandom())
  n11 = ScalarParameter() {defaultValue=120, parameterName=Flap Speed}
  n12 = Multiply(A=Sine(None=Multiply(A=@n10, B=@n11)), B=Multiply(A=Constant() {r=4}, B=ScalarParameter() {defaultValue=1, parameterName=Flap Amount}))
  Multiply(A=LinearInterpolate(A=@n3, B=Add(A=@n3, B=@n4.RGB), Alpha=Multiply(A=@n9, B=@n12)))
  Transform() {transformSourceType=TRANSFORMSOURCE_World, transformType=TRANSFORM_Instance}
  VertexColor()
 ]
}
