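# Two-sided lit material for the Magic Forest firefly body mesh, usable on Niagara mesh particles and instanced static meshes.
# Particle color tints diffuse and emissive textures, with a time-of-day blend brightening night glow.
# GlowColor dynamic parameter scales emission, including a directional normal-based term, with a separate ray-tracing multiplier.
# A WPO function sits alongside sine flap nodes using particle random, vertex normals and red vertex-color masking.
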Material /Game/Effects/Fort_Effects/Effects/Environment/MagicForest/Materials/M_Firefly_Body
graph {
 properties {
  bEnableResponsiveAA=true
  bUsedWithInstancedStaticMeshes=true
  bUsedWithNiagaraMeshParticles=true
  pixelDepthOffsetMode=PDOM_Legacy
  twoSided=true
  usedShadingModels=Default Lit
 }
 nodes [
  CameraPositionWS()
  Clamp(None=Sine(None=Time()))
  Constant3Vector() {constant=<rgba 22.9987 0 60 1>}
  n1 = ParticleColor()
  n2 = Multiply(A=@n1.RGB) {constB=2}
  n3 = CollectionParameter() {
   collection=&FortniteMaterialParameters
   parameterName=FortniteActiveTimeOfDay
  }
  n4 = DotProduct(A=@n3, B=Constant4Vector() {constant=<rgba 1 1 1 0>}.RGBA)
  n5 = DynamicParameter() {defaultValue=<rgba 0 0 0 1>, paramNames [Param1,Param2,Param3,GlowColor]}
  n6 = Multiply(
   A=TextureSample() {texture=/Game/Effects/Fort_Effects/Textures/Fireflies/T_Firefly_Mesh_01_e}.RGB
   B=Multiply(A=LinearInterpolate(A=@n2, B=@n1.RGB, Alpha=@n4) {desc="A=Night B=Day"}, B=@n5.GlowColor)
  )
  n7 = Multiply(A=@n6, B=ScalarParameter())
  n8 = DotProduct(
   A=VertexInterpolator(VS=PreSkinnedNormal()).PS
   B=Transform(None=VectorParameter() {defaultValue=<rgba -1 0 1 0>}.RGB) {transformSourceType=TRANSFORMSOURCE_World, transformType=TRANSFORM_Instance}
  )
  n9 = Multiply(A=ScalarParameter() {parameterName=Param_1}, B=Multiply(A=@n1.RGB, B=Saturate(None=@n8)))
  MaterialFunctionCall(Emissive=Add(A=RayTracingQualitySwitch(Normal=@n6, RayTraced=@n7), B=Multiply(A=@n5.GlowColor, B=@n9))) {materialFunction=&MF_FX_EmissiveAdjustments}
  MaterialFunctionCall() {materialFunction=&/M_Firefly_Body_WPO_Func}
  MaterialFunctionCall(DiffuseColor=Multiply(A=TextureSample() {texture=/Game/Effects/Fort_Effects/Textures/Fireflies/T_Firefly_Mesh_01}.RGB, B=@n1.RGB)) {materialFunction=&ConvertFromDiffSpec}
  Multiply(B=MaterialFunctionCall() {materialFunction=/Engine/Functions/Engine_MaterialFunctions02/WorldPositionOffset/ObjectScale}.Scale XYZ)
  n10 = VertexColor()
  n11 = Add(A=Time(), B=ParticleRandom())
  n12 = ScalarParameter() {defaultValue=120, parameterName=Flap Speed}
  n13 = Multiply(A=Sine(None=Multiply(A=@n11, B=@n12)), B=Multiply(A=Constant() {r=4}, B=ScalarParameter() {defaultValue=2, parameterName=Flap Amount}))
  Multiply(A=Multiply(A=VertexNormalWS(), B=Multiply(A=@n10.R, B=@n13)))
  VertexNormalWS()
 ]
}
