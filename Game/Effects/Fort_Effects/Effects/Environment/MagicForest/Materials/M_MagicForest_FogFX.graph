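# Volume-domain material for Magic Forest fog effects, drawing soft volumetric fog puffs from Niagara particles.
# A soft sphere mask at each particle's position and radius is eroded by a curl noise texture, sampled with time-scrolled world coordinates distorted by procedural curl noise.
# A noise-offset height cutoff relative to the particle and near-camera depth fades shape the density.
# Particle color and atmospheric light color tint the result, and particle alpha scales opacity.
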Material /Game/Effects/Fort_Effects/Effects/Environment/MagicForest/Materials/M_MagicForest_FogFX
graph {
 properties {
  bAutomaticallySetUsageInEditor=false
  bUsedWithNiagaraSprites=true
  blendMode=BLEND_Additive
  materialDomain=MD_Volume
  usedShadingModels=Default Lit
 }
 nodes [
  Add()
  n1 = ParticleColor()
  n2 = SphereMask(A=WorldPosition().XYZ, B=ParticlePositionWS(), Radius=ParticleRadius()) {hardnessPercent=0}
  n3 = WorldPosition()
  n4 = VectorNoise(World Position=Multiply(A=@n3.XYZ, B=ScalarParameter() {defaultValue=0.0009, parameterName=Param_5})) {noiseFunction=VNF_CurlALU}
  n5 = ScalarParameter() {defaultValue=0.075, parameterName=Param_4}
  n6 = Add(A=@n3.XYZ, B=Multiply(A=Time(), B=VectorParameter() {defaultValue=<rgba -62 0 0 0>}.RGB))
  n7 = ScalarParameter() {defaultValue=0.0005}
  n8 = TextureSample(UVs=Add(A=Multiply(A=@n4, B=@n5), B=Multiply(A=@n6, B=@n7))) {samplerType=SAMPLERTYPE_LinearColor, texture=/Game/Effects/FogVolume/VT_CurlNoise_LOW}
  n9 = Multiply(A=@n8.RGB, B=ScalarParameter() {defaultValue=10, parameterName=Param_2})
  n10 = Subtract(A=ParticlePositionWS(), B=WorldPosition().XYZ)
  n11 = Subtract(A=ComponentMask(None=Add(A=@n9, B=@n10)) {b=true}, B=ScalarParameter() {defaultValue=-150, parameterName=Param_1})
  n12 = PixelDepth()
  n13 = ScalarParameter() {defaultValue=1500, parameterName=Near Camera Fade}
  n14 = Multiply(A=Multiply(A=Subtract(A=@n2, B=@n8.G), B=Saturate(None=@n11)), B=Saturate(None=Divide(A=@n12.R, B=@n13)))
  n15 = PixelDepth()
  n16 = ScalarParameter() {defaultValue=1500, parameterName=Near Camera Fade}
  Multiply(A=Multiply(A=@n1.A, B=@n14), B=Saturate(None=Divide(A=@n15.R, B=@n16)))
  n17 = AtmosphericLightColor()
  n18 = ComponentMask(None=@n14) {g=true}
  Multiply(A=Multiply(A=@n1.RGB, B=Multiply(A=@n17, B=@n18)), B=ScalarParameter() {defaultValue=0.0025, parameterName=Param_3})
 ]
}
