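# Translucent, unlit, two-sided material for glowing spline trail meshes in Magic Forest effects, used on Niagara mesh particles.
# SplineThicken sets ribbon width between Tip and End.
# An offset from dynamic parameter X, weighted along inverted V, transformed to instance space and scaled by object scale, is added along with world-space curl noise to displace vertices.
# A V-based gradient, peaking mid-length and multiplied by particle alpha, fades both ends.
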Material /Game/Effects/Fort_Effects/Effects/Environment/MagicForest/Materials/M_Spline_GlowTrail
graph {
 properties {
  bAutomaticallySetUsageInEditor=false
  bUsedWithNiagaraMeshParticles=true
  blendMode=BLEND_Translucent
  shadingModel=MSM_Unlit
  twoSided=true
  usedShadingModels=Unlit
 }
 nodes [
  n1 = ScalarParameter() {defaultValue=5, parameterName=Tip}
  n2 = ScalarParameter() {parameterName=End}
  n3 = WorldPosition()
  n4 = DynamicParameter() {paramNames [X,Y,Z,Param4]}
  n5 = MaterialFunctionCall(X=@n4.X, Y=@n4.X, Z=@n4.X) {materialFunction=&MakeFloat3}
  n6 = ScalarParameter() {defaultValue=2, parameterName=Param_3}
  n7 = OneMinus(None=ComponentMask(None=TextureCoordinate()) {g=true})
  n8 = ScalarParameter() {defaultValue=4, parameterName=Param_1}
  n9 = LinearInterpolate(
   B=Transform(None=Multiply(A=@n5.Result, B=@n6)) {transformSourceType=TRANSFORMSOURCE_World, transformType=TRANSFORM_Instance}
   Alpha=Saturate(None=Power(Base=@n7, Exp=@n8))
  )
  n10 = MaterialFunctionCall() {materialFunction=&ObjectScale}
  n11 = Multiply(A=WorldPosition().XYZ, B=ScalarParameter() {defaultValue=0.005, parameterName=Param_4})
  Add(
   A=Add(
    A=MaterialFunctionCall(WidthBase=@n1, WidthTip=@n2, WorldPosition=@n3.XYZ) {materialFunction=&SplineThicken}.WorldPositionOffset
    B=Multiply(A=@n9, B=@n10.Scale XYZ)
   )
   B=Multiply(
    A=ScalarParameter() {defaultValue=0.25, parameterName=Param_6}
    B=TextureSample(UVs=@n11) {samplerType=SAMPLERTYPE_LinearColor, texture=/Game/Effects/FogVolume/VT_CurlNoise_LOW}.RGB
   )
  )
  n12 = ComponentMask(None=TextureCoordinate()) {g=true}
  n13 = OneMinus(None=@n12)
  Multiply(A=ParticleColor().A, B=Multiply(A=Multiply(A=@n13, B=@n12), B=ScalarParameter() {defaultValue=5, parameterName=Param_5}))
  Normalize()
  Saturate(None=Multiply(A=OneMinus(), B=ScalarParameter() {defaultValue=1, parameterName=Param_2}))
  Saturate(None=Add())
  ScalarParameter() {defaultValue=5}
 ]
}
